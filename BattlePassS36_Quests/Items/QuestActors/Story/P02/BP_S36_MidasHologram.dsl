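; Season 36 story quest actor for a Midas hologram the player talks to, shown through a child-actor mannequin.
; Gameplay-tag side effects drive its states: Appear caches the mannequin and plays a wave on remotes, while Waiting does nothing.
; Interacted binds conversation end to broadcast the P02 Q05 ConversationComplete custom verb.
; Complete plays a pointing emote on remotes, while authority spawns a poof Niagara effect after a delay.
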(event EventBeginPlay
  (|CallParentFunction) ; This node is disabled and will not be called.\nDrag off pins to build functionality.
)

(event Collision|EventActorBeginOverlap
  (params (OtherActor :type Actor%20Object%20Reference))
  (|CallParentFunction OtherActor) ; This node is disabled and will not be called.\nDrag off pins to build functionality.
)

(event EventTick
  (params (DeltaSeconds :type Float%20%28single-precision%29))
  (|CallParentFunction DeltaSeconds) ; This node is disabled and will not be called.\nDrag off pins to build functionality.
)

(event Custom|ConvoEnd
  ;[0 0.116 0.15 0.5] State Handling
  ; Appear - Brief moment when mannequin first appears
  ; Waiting - Mannequin is wating to be spoken to
  ; Interacted - player has interacted
  ; Complete - the interaction is complete and the actor is about to be hidden
  (CustomVerb|BroadcastCustomVerbMessage (CustomVerb|MakeCustomVerbMessage self "(TagName=\"CustomVerb.Quest.S36.P02.Q05.ConversationComplete\")")))

(event SideEffects|EventBP_ApplySideEffect
  (params (ReactionSource :type Object%20Reference) (TargetObject :type Object%20Reference) (SideEffectTriggerTag :type Gameplay%20Tag%20Structure%20%28by%20ref%29) (SideEffectTriggerType :type EBlueprintableSideEffectTriggerType%20Enum%20%28by%20ref%29 :default "TriggeredAsTarget"))
  ;[0 0.116 0.15 0.5] State Handling
  ; Appear - Brief moment when mannequin first appears
  ; Waiting - Mannequin is wating to be spoken to
  ; Interacted - player has interacted
  ; Complete - the interaction is complete and the actor is about to be hidden
  (switch Utilities|FlowControl|Switch|SwitchonGameplayTag SideEffectTriggerTag
    (:Quests.S36.SDA.BPSideEffect.Appear
      (Networking|SwitchHasAuthority
        (:Remote
          (bind _asbp_s36_midas_mannequin (Utilities|Casting|CastToBP_S36_MidasMannequin (Variables|/Script/Engine.ChildActorComponent|GetChildActor (Variables|Default|GetBP_S36_MidasMannequin))))
          (Variables|Default|SetAs%20BP%20S36%20Midas%20Mannequin _asbp_s36_midas_mannequin)
          (Components|Animation|PlayAnimation (Variables|/Script/Engine.SkeletalMeshActor|GetSkeletalMeshComponent _asbp_s36_midas_mannequin) "/BRCosmetics/Animation/Game/MainPlayer/Emotes/Emote_Wave2_CMM.Emote_Wave2_CMM" false))))
    (:Quests.S36.SDA.BPSideEffect.Waiting)
    (:Quests.S36.SDA.BPSideEffect.Interacted
      (EventDispatchers|BindEventtoOnConversationEnded (Variables|Default|GetNonPlayerConversationComponent) (delegate Custom|ConvoEnd)))
    (:Quests.S36.SDA.BPSideEffect.Complete
      (Networking|SwitchHasAuthority
        (:Authority
          (Utilities|FlowControl|Delay 3.75)
          (CallFunction|PoofVFX))
        (:Remote
          (Components|Animation|PlayAnimation (Variables|/Script/Engine.SkeletalMeshActor|GetSkeletalMeshComponent (Variables|Default|GetAs%20BP%20S36%20Midas%20Mannequin)) "/BRCosmetics/Animation/Game/MainPlayer/Emotes/PapayaComms/PapayaComms_PointFinger_CMM.PapayaComms_PointFinger_CMM" false))))))

(event Custom|PoofVFX
  ;[0 0.116 0.15 0.5] State Handling
  ; Appear - Brief moment when mannequin first appears
  ; Waiting - Mannequin is wating to be spoken to
  ; Interacted - player has interacted
  ; Complete - the interaction is complete and the actor is about to be hidden
  ;; POOF FX
  (Niagara|SpawnSystemAttached "/Game/Athena/Items/QuestInteractables/Generic/FX/NS_Athena_QuestItem_PickupSuccess.NS_Athena_QuestItem_PickupSuccess" (Variables|Default|GetBP_S36_MidasMannequin) "FX_Feet"))

(fn ConstructionScript
  (|CallParentFunction))
