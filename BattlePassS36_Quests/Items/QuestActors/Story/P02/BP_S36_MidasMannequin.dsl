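; Skeletal mesh actor Blueprint for a Midas mannequin in the Season 36 Battle Pass story quests.
; The construction script overrides material slot 0 with the MistRaven Wraith Form quest material.
; Its LoadAndPlayMontage event async-loads a soft montage reference, casts and validates it, stores it in MontageActual, then plays it looping on the mesh.
; BeginPlay is empty, and nothing in this graph calls the montage event.
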(event EventBeginPlay)

(event Custom|LoadAndPlayMontage
  (params (LocalMontage :type Anim%20Montage%20Soft%20Object%20Reference))
  (bind _object (Utilities|AsyncLoadAsset LocalMontage)
    (:Completed
      (bind _asanim_montage (Utilities|Casting|CastToAnimMontage _object))
      (Utilities|IsValid _asanim_montage
        (:Is%20Valid
          (bind _output_get (Variables|Default|SetMontageActual _asanim_montage))
          (bind _skeletalmeshcomponent (Variables|SkeletalMeshActor|GetSkeletalMeshComponent)
            (:then
              (Components|Animation|PlayAnimation _skeletalmeshcomponent _output_get true))
            (:else)))))))

(fn ConstructionScript
  (Rendering|Material|SetMaterial (Variables|SkeletalMeshActor|GetSkeletalMeshComponent) 0 "/BRCosmetics/Characters/Player/Female/Medium/Bodies/F_MED_MistRaven_Peck/Skins/Quest/Materials/M_MistRaven_WraithForm_Quest.M_MistRaven_WraithForm_Quest"))
