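; Season 36 story quest target dummy actor whose interactability and feedback are driven by gameplay-tagged side effects.
; Side effect B enables query-only collision on the dummy's top mesh and starts an attached ambient loop, while A and C disable that collision.
; The Interacted effect stops the loop, plays a hit cue and runs a timeline that rotates the top mesh while setting health to a fixed value each update.
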(event EventBeginPlay
  (|CallParentFunction) ; This node is disabled and will not be called.\nDrag off pins to build functionality.
)

(event Collision|EventActorBeginOverlap
  (params (OtherActor :type Actor%20Object%20Reference))
  (|CallParentFunction OtherActor) ; This node is disabled and will not be called.\nDrag off pins to build functionality.
)

(event EventTick
  (params (DeltaSeconds :type Float%20%28single-precision%29))
  (|CallParentFunction DeltaSeconds) ; This node is disabled and will not be called.\nDrag off pins to build functionality.
)

(event SideEffects|EventBP_ApplySideEffect
  (params (ReactionSource :type Object%20Reference) (TargetObject :type Object%20Reference) (SideEffectTriggerTag :type Gameplay%20Tag%20Structure%20%28by%20ref%29) (SideEffectTriggerType :type EBlueprintableSideEffectTriggerType%20Enum%20%28by%20ref%29 :default "TriggeredAsTarget"))
  (switch Utilities|FlowControl|Switch|SwitchonGameplayTag SideEffectTriggerTag
    (:Quests.S36.SDA.BPSideEffect.A
      (Collision|SetCollisionEnabled (Variables|Default|GetSM_Quest_TrainingDummyTop_A) "NoCollision"))
    (:Quests.S36.SDA.BPSideEffect.B
      (Collision|SetCollisionEnabled (Variables|Default|GetSM_Quest_TrainingDummyTop_A) "QueryOnly")
      (bind _returnvalue (Audio|SpawnSoundAttached "/Game/Athena/Sounds/Quests/Athena_HiddenStar_AmbientLoop_Cue.Athena_HiddenStar_AmbientLoop_Cue" (Variables|Default|GetStaticMesh)))
      (Variables|Default|SetSoundloop3 _returnvalue))
    (:Quests.S36.SDA.BPSideEffect.C
      (Collision|SetCollisionEnabled (Variables|Default|GetSM_Quest_TrainingDummyTop_A) "NoCollision"))
    (:Quests.S36.SDA.BPSideEffect.Interacted
      (Audio|Components|Audio|Stop (Variables|Default|GetSoundloop3))
      (Audio|SpawnSoundAttached "/BattlePassS36_Quests/Audio/SFX/Quest_S36_Story_MistRaven_TargetDummy/Quest_S36_Story_MistRaven_TargetDummy_Cue.Quest_S36_Story_MistRaven_TargetDummy_Cue" (Variables|Default|GetSM_Quest_TrainingDummyTop_A))
      (bind (_direction _height _rotation) (Timeline|Timeline PlayFromStart)
        (:Update
          (Transformation|SetRelativeTransform (Variables|Default|GetSM_Quest_TrainingDummyTop_A) (Math|Transform|MakeTransform "0, 0, 0" (Math|Rotator|MakeRotator (.x _rotation) (.y _rotation) (.z _rotation))) false false)
          (Health|SetHealth 2.5e+02))))))

(fn ConstructionScript
  (|CallParentFunction))
